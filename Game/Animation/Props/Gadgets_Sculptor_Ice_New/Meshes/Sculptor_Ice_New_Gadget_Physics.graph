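# Physics asset for the Sculptor Ice New gadget prop, giving its mesh a simplified collision shape.
# A single body on bone JNT_Year_24 uses one capsule, offset from the bone and rotated.
# It has no constraints, so there is no jointed ragdoll chain.
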PhysicsAsset /Game/Animation/Props/Gadgets_Sculptor_Ice_New/Meshes/Sculptor_Ice_New_Gadget_Physics
bodies {JNT_Year_24 [{shapeType=Capsule, length=40.226, radius=32.7809, center=<-3.6657 0.019 35.6055>, rotation=<rot 0.0206 -90 -10.5726>}]}
constraints []
